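# Niagara sprite effect for track Overdrive, spawning short-lived particles along both sides of the track.
# One emitter partitions particles left and right in a long box, pushes them inward against drag, and lays sprites facing up.
# User.Is5Lane lerps the side offset wider, and User.IsPlasticOverdrive switches the color from orange to white.
# Particles pop in and shrink by curve, with alpha scaled by spawn position along the track.
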NiagaraSystem /PilgrimCore/GemAndTrack/Track/Particles/NS_Track_Overdrive
dependencies {
 usedDataInterfaces [&NiagaraDataInterfaceCurve]
 usedDynamicInputs []
 usedModules [
  &self:SystemSpawnScript.SystemScriptSource.SystemScriptGraph.NiagaraNodeAssignment_0.SetVariables_#1
  &SystemState
 ]
}
emitters {Side_Particles001 {
  emitterUpdateScript [
   &EmitterState(Life Cycle Mode=::NewEnumerator1, Loop Behavior=::NewEnumerator1, Loop Duration=0.25, UseLoopDelay=false)
   &SpawnRate(SpawnRate=50, Use Spawn Probability=false)
  ]
  particleSpawnScript [
   &PartitionParticles()
   &InitializeParticle(
    Color=&MakeCustomLinearColorFromBool(Bool=$User.IsPlasticOverdrive, True Color=<rgba 2 2 2 1>, False Color=<rgba 4 1.4266 0.1458 1>)
    Lifetime=&RandomRangeFloat(Minimum=0.5, Maximum=1)
    Position Offset=&Lerp_Vector(
     A=&Lerp_Vector(A=<-9.75 2 0>, B=<9.75 2 0>, Alpha=&MakeFloatFromInt(Int=$Particles.PartitionParticles.Partition))
     B=&Lerp_Vector(A=<-11.5 2 0>, B=<11.5 2 0>, Alpha=&MakeFloatFromInt(Int=$Particles.PartitionParticles.Partition))
     Alpha=$User.Is5Lane
    )
    Sprite Size=<5 20>
    Sprite Size Mode=::NewEnumerator4
    UsePositionOffset=true
    Write Material Random=true
   )
   &ShapeLocation(
    Box Midpoint=<0.5 1.5 0.5>
    Box Size=<0 70 0>
    MaskSpawnGroup=false
    Override Seed=false
    Shape Primitive=::NewEnumerator2
    Surface Only=false
    UseOwnerScale=true
   )
   &AddVelocity(
    Velocity=&Lerp_Vector(
     A=&RandomRangeVector(Minimum=<5 0 0>, Maximum=<10 0 0>)
     B=&RandomRangeVector(Minimum=<-10 0 0>, Maximum=<-5 0 0>)
     Alpha=&MakeFloatFromInt(Int=$Particles.PartitionParticles.Partition)
    )
    Velocity Speed Scale=4
   ) {module=AddVelocity001}
   &SpriteFacingAndAlignment(Sprite Facing=<0 0 1>, WriteAlignment=false, WriteFacing=true)
   &DynamicMaterialParameters(
    Index 0 Param 1=0.8
    Module.Index0Param0Write=true
    Module.Index0Param1Write=true
    Module.Index0Param2Write=true
    Module.Index0Param3Write=true
   )
  ]
  particleUpdateScript [
   &ParticleState()
   &ScaleSpriteSize(
    Uniform Curve Scale=0.5
    Uniform Curve Sprite Scale=DataInterface() {propertyValues {Curve {keys [
        {arriveTangent=9.0725, interpMode=RCIM_Cubic, leaveTangent=9.0725, tangentMode=RCTM_User, time=0, value=0}
        {arriveTangent=-2.5202, interpMode=RCIM_Cubic, leaveTangent=-2.5202, tangentMode=RCTM_User, time=0.1408, value=0.9847}
        {arriveTangent=-0.0405, interpMode=RCIM_Cubic, leaveTangent=-0.0405, tangentMode=RCTM_User, time=1, value=0}
       ]}}}
   )
   &Drag(Drag=5, Use Linear Drag=true, Use Rotational Drag=true)
   &SolveForcesAndVelocity(Clamp Velocity=false, Limit Acceleration=false)
   &ScaleColor(
    Scale Alpha=&ClampFloat(Float=&Subtract_Float(
      A=1
      B=&NormalizeFloat(
       Value=&MakeFloatFromVector(Vector=$Particles.ShapeLocation.ShapePosition, Channel=::NewEnumerator1)
       Min=1
       Max=-110
      )
     ))
    ScaleA=true
    ScaleRGB=true
   )
  ]
  rendererClasses [&NiagaraSpriteRendererProperties]
  renderers [{rendererClass=&NiagaraSpriteRendererProperties}]
 }}
summary {
 systemName=NS_Track_Overdrive
 emitters [{emitterName=Side_Particles001, rendererClasses [&NiagaraSpriteRendererProperties]}]
 userVariables [
  User.Is5Lane {type=NiagaraFloat, defaultValue=0, description=""}
  User.IsPlasticOverdrive {type=NiagaraBool}
 ]
}
