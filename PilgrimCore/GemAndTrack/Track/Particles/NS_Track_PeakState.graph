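# Niagara system for the track's peak state effect: a blue sprite burst at the track center.
# Five burst emitters layer a swelling, fading corona, a quick-pop flare, flare sparks spreading along X, and left and right streak sparks slowed by drag.
# All sprites face up along Z, and every emitter's X offset lerps from zero to System.5LaneCenterLocation by User.Is5Lane, re-centering the effect for five-lane tracks.
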NiagaraSystem /PilgrimCore/GemAndTrack/Track/Particles/NS_Track_PeakState
dependencies {
 usedDataInterfaces [&NiagaraDataInterfaceCurve, &NiagaraDataInterfaceColorCurve, &NiagaraDataInterfaceVector2DCurve]
 usedDynamicInputs []
 usedModules [
  &self:SystemSpawnScript.SystemScriptSource.SystemScriptGraph.NiagaraNodeAssignment_0.SetVariables_#1
  &SystemState
 ]
}
emitters {
 Center_Burst_Sparks_Flare {
  emitterUpdateScript [
   &EmitterState()
   &SpawnBurst_Instantaneous(Spawn Count=20, Use Spawn Probability=false, UseLoopCountLimit=false)
  ]
  particleSpawnScript [
   &InitializeParticle(
    Color=<rgba 0.2 0.2 4 1>
    Lifetime=&RandomRangeFloat(Minimum=0.5, Maximum=1)
    Position Offset=&MakeVector(X=&Lerp_Float(A=0, B=$System.5LaneCenterLocation, Alpha=$User.Is5Lane), Y=-6, Z=0)
    Sprite Size=<5 20>
    Sprite Size Mode=::NewEnumerator4
    UsePositionOffset=true
    Write Material Random=true
   )
   &ShapeLocation(Box Size=<0 3 0>, MaskSpawnGroup=false, Override Seed=false, Shape Primitive=::NewEnumerator2, Surface Only=false, UseOwnerScale=true)
   &AddVelocity(Velocity=&RandomRangeVector(Minimum=<-40 0 0>, Maximum=<40 0 0>)) {module=AddVelocity001}
   &SpriteFacingAndAlignment(Sprite Facing=<0 0 1>, WriteAlignment=false, WriteFacing=true)
  ]
  particleUpdateScript [
   &ParticleState()
   &ScaleSpriteSize(
    Uniform Curve Scale=2
    Uniform Curve Sprite Scale=DataInterface() {propertyValues {Curve {keys [
        {arriveTangent=9.0725, interpMode=RCIM_Cubic, leaveTangent=9.0725, tangentMode=RCTM_User, time=0, value=0}
        {arriveTangent=-2.5202, interpMode=RCIM_Cubic, leaveTangent=-2.5202, tangentMode=RCTM_User, time=0.1408, value=0.9847}
        {arriveTangent=-0.0405, interpMode=RCIM_Cubic, leaveTangent=-0.0405, tangentMode=RCTM_User, time=1, value=0}
       ]}}}
   )
   &Drag(Drag=5, Use Linear Drag=true, Use Rotational Drag=true)
   &SolveForcesAndVelocity(Clamp Velocity=false, Limit Acceleration=false)
   &ScaleColor(ScaleA=true, ScaleRGB=true)
  ]
  rendererClasses [&NiagaraSpriteRendererProperties]
  renderers [{rendererClass=&NiagaraSpriteRendererProperties}]
 }
 Center_Burst_Sparks_Left {
  emitterUpdateScript [
   &EmitterState()
   &SpawnBurst_Instantaneous(Spawn Count=5, Use Spawn Probability=false, UseLoopCountLimit=false)
  ]
  particleSpawnScript [
   &InitializeParticle(
    Color=<rgba 0.2 0.2 4 1>
    Lifetime=&RandomRangeFloat(Minimum=0.35, Maximum=0.65)
    Position Offset=&MakeVector(X=&Lerp_Float(A=0, B=$System.5LaneCenterLocation, Alpha=$User.Is5Lane), Y=-6, Z=0)
    Sprite Size=&RandomRangeVector2D(Minimum=<2 1>, Maximum=<2 3>)
    Sprite Size Mode=::NewEnumerator4
    UsePositionOffset=true
    Write Material Random=true
   )
   &ShapeLocation(Box Size=<0 3 0>, MaskSpawnGroup=false, Override Seed=false, Shape Primitive=::NewEnumerator2, Surface Only=false, UseOwnerScale=true)
   &AddVelocity(Velocity=&RandomRangeVector(Minimum=<-40 0 0>, Maximum=<-20 1 1>)) {module=AddVelocity001}
   &SpriteFacingAndAlignment(Sprite Facing=<0 0 1>, WriteAlignment=false, WriteFacing=true)
   &DynamicMaterialParameters(
    Index 0 Param 1=0.8
    Module.Index0Param0Write=true
    Module.Index0Param1Write=true
    Module.Index0Param2Write=true
    Module.Index0Param3Write=true
   )
  ]
  particleUpdateScript [
   &ParticleState()
   &ScaleSpriteSize(
    Non-Uniform Curve Sprite Scale=DataInterface() {propertyValues {
      XCurve {keys [{tangentMode=RCTM_User, time=0, value=1}, {tangentMode=RCTM_User, time=1, value=1}]}
      YCurve {keys [
        {interpMode=RCIM_Cubic, tangentMode=RCTM_User, time=0, value=3}
        {arriveTangent=-6.2824, interpMode=RCIM_Cubic, leaveTangent=-6.2824, tangentMode=RCTM_User, time=1, value=0.25}
       ]}
     }}
    Scale Sprite Size Mode=::NewEnumerator3
   )
   &Drag(Drag=5, Use Linear Drag=true, Use Rotational Drag=true)
   &SolveForcesAndVelocity(Clamp Velocity=false, Limit Acceleration=false)
   &ScaleColor(ScaleA=true, ScaleRGB=true)
  ]
  rendererClasses [&NiagaraSpriteRendererProperties]
  renderers [{rendererClass=&NiagaraSpriteRendererProperties}]
 }
 Center_Burst_Sparks_Right {
  emitterUpdateScript [
   &EmitterState()
   &SpawnBurst_Instantaneous(Spawn Count=5, Use Spawn Probability=false, UseLoopCountLimit=false)
  ]
  particleSpawnScript [
   &InitializeParticle(
    Color=<rgba 0.2 0.2 4 1>
    Lifetime=&RandomRangeFloat(Minimum=0.35, Maximum=0.65)
    Position Offset=&MakeVector(X=&Lerp_Float(A=0, B=$System.5LaneCenterLocation, Alpha=$User.Is5Lane), Y=-6, Z=0)
    Sprite Size=&RandomRangeVector2D(Minimum=<2 1>, Maximum=<2 3>)
    Sprite Size Mode=::NewEnumerator4
    UsePositionOffset=true
    Write Material Random=true
   )
   &ShapeLocation(Box Size=<0 3 0>, MaskSpawnGroup=false, Override Seed=false, Shape Primitive=::NewEnumerator2, Surface Only=false, UseOwnerScale=true)
   &AddVelocity(Velocity=&RandomRangeVector(Minimum=<20 0 0>, Maximum=<40 1 1>)) {module=AddVelocity001}
   &SpriteFacingAndAlignment(Sprite Facing=<0 0 1>, WriteAlignment=false, WriteFacing=true)
   &DynamicMaterialParameters(
    Index 0 Param 1=0.8
    Module.Index0Param0Write=true
    Module.Index0Param1Write=true
    Module.Index0Param2Write=true
    Module.Index0Param3Write=true
   )
  ]
  particleUpdateScript [
   &ParticleState()
   &ScaleSpriteSize(
    Non-Uniform Curve Sprite Scale=DataInterface() {propertyValues {
      XCurve {keys [{tangentMode=RCTM_User, time=0, value=1}, {tangentMode=RCTM_User, time=1, value=1}]}
      YCurve {keys [
        {interpMode=RCIM_Cubic, tangentMode=RCTM_User, time=0, value=3}
        {arriveTangent=-6.2824, interpMode=RCIM_Cubic, leaveTangent=-6.2824, tangentMode=RCTM_User, time=1, value=0.25}
       ]}
     }}
    Scale Sprite Size Mode=::NewEnumerator3
   )
   &Drag(Drag=5, Use Linear Drag=true, Use Rotational Drag=true)
   &SolveForcesAndVelocity(Clamp Velocity=false, Limit Acceleration=false)
   &ScaleColor(ScaleA=true, ScaleRGB=true)
  ]
  rendererClasses [&NiagaraSpriteRendererProperties]
  renderers [{rendererClass=&NiagaraSpriteRendererProperties}]
 }
 Center_Corona {
  emitterUpdateScript [
   &EmitterState()
   &SpawnBurst_Instantaneous(Use Spawn Probability=false, UseLoopCountLimit=false)
  ]
  particleSpawnScript [
   &InitializeParticle(
    Position Offset=&MakeVector(X=&Lerp_Float(A=0, B=$System.5LaneCenterLocation, Alpha=$User.Is5Lane), Y=-6.5, Z=0)
    Sprite Size=<15 40>
    Sprite Size Mode=::NewEnumerator4
    UsePositionOffset=true
    Write Material Random=true
   )
   &SpriteFacingAndAlignment(Sprite Facing=<0 0 1>, WriteAlignment=false, WriteFacing=true)
  ]
  particleUpdateScript [
   &ParticleState()
   &ScaleSpriteSize(
    Uniform Curve Scale=4
    Uniform Curve Sprite Scale=DataInterface() {propertyValues {Curve {keys [
        {arriveTangent=4.1785, interpMode=RCIM_Cubic, leaveTangent=4.1785, tangentMode=RCTM_User, time=0, value=0}
        {arriveTangent=1.1179, interpMode=RCIM_Cubic, leaveTangent=1.1179, tangentMode=RCTM_User, time=0.2042, value=0.8}
        {arriveTangent=0.0224, interpMode=RCIM_Cubic, leaveTangent=0.0224, tangentMode=RCTM_User, time=1, value=1}
       ]}}}
   )
   &Color(Color=&MultiplyLinearColor(
     Linear Color=&ColorFromCurve(ColorCurve=DataInterface() {propertyValues {
        AlphaCurve {keys [<at 0 1 RCIM_Cubic>,<at 1 1 RCIM_Cubic>]}
        BlueCurve {keys [<at 0 1 RCIM_Cubic>,<at 0.417 0.25 RCIM_Cubic>]}
        GreenCurve {keys [<at 0 0.1625 RCIM_Cubic>,<at 0.417 0 RCIM_Cubic>]}
        RedCurve {keys [<at 0 0 RCIM_Cubic>,<at 0.417 0.0417 RCIM_Cubic>]}
       }})
     Scale Factor=&MakeVector4(
      X=20
      Y=20
      Z=20
      W=&FloatFromCurve(FloatCurve=DataInterface() {propertyValues {Curve {keys [
           {arriveTangent=-2.9379, interpMode=RCIM_Cubic, leaveTangent=-2.9379, tangentMode=RCTM_User, time=0, value=1}
           {interpMode=RCIM_Cubic, tangentMode=RCTM_User, time=1, value=0}
          ]}}})
     )
    ))
  ]
  rendererClasses [&NiagaraSpriteRendererProperties]
  renderers [{rendererClass=&NiagaraSpriteRendererProperties}]
 }
 Center_Flare {
  emitterUpdateScript [
   &EmitterState()
   &SpawnBurst_Instantaneous(Use Spawn Probability=false, UseLoopCountLimit=false)
  ]
  particleSpawnScript [
   &InitializeParticle(
    Color=<rgba 0 0.0406 0.25 1>
    Position Offset=&MakeVector(X=&Lerp_Float(A=0, B=$System.5LaneCenterLocation, Alpha=$User.Is5Lane), Y=-6.5, Z=0)
    Sprite Size=<20 40>
    Sprite Size Mode=::NewEnumerator4
    UsePositionOffset=true
    Write Material Random=true
   )
   &SpriteFacingAndAlignment(Sprite Facing=<0 0 1>, WriteAlignment=false, WriteFacing=true)
  ]
  particleUpdateScript [
   &ParticleState()
   &ScaleSpriteSize(
    Uniform Curve Scale=6
    Uniform Curve Sprite Scale=DataInterface() {propertyValues {Curve {keys [
        {arriveTangent=34.0082, interpMode=RCIM_Cubic, leaveTangent=34.0082, tangentMode=RCTM_User, time=0, value=0}
        {arriveTangent=-2.5202, interpMode=RCIM_Cubic, leaveTangent=-2.5202, tangentMode=RCTM_User, time=0.05, value=0.9847}
        {arriveTangent=-0.0405, interpMode=RCIM_Cubic, leaveTangent=-0.0405, tangentMode=RCTM_User, time=1, value=0}
       ]}}}
   )
   &Color(Color=&MultiplyLinearColor(
     Linear Color=&ColorFromCurve(ColorCurve=DataInterface() {propertyValues {
        AlphaCurve {keys [<at 0 1 RCIM_Cubic>,<at 1 1 RCIM_Cubic>]}
        BlueCurve {keys [<at 0 1 RCIM_Cubic>,<at 0.417 0.25 RCIM_Cubic>]}
        GreenCurve {keys [<at 0 0.1625 RCIM_Cubic>,<at 0.417 0 RCIM_Cubic>]}
        RedCurve {keys [<at 0 0 RCIM_Cubic>,<at 0.417 0.108 RCIM_Cubic>]}
       }})
     Scale Factor=&MakeVector4(X=5, Y=5, Z=5, W=1)
    ))
  ]
  rendererClasses [&NiagaraSpriteRendererProperties]
  renderers [{rendererClass=&NiagaraSpriteRendererProperties}]
 }
}
summary {
 systemName=NS_Track_PeakState
 emitters [
  {emitterName=Center_Burst_Sparks_Flare, rendererClasses [&NiagaraSpriteRendererProperties]}
  {emitterName=Center_Flare, rendererClasses [&NiagaraSpriteRendererProperties]}
  {emitterName=Center_Burst_Sparks_Right, rendererClasses [&NiagaraSpriteRendererProperties]}
  {emitterName=Center_Corona, rendererClasses [&NiagaraSpriteRendererProperties]}
  {emitterName=Center_Burst_Sparks_Left, rendererClasses [&NiagaraSpriteRendererProperties]}
 ]
 userVariables [User.Is5Lane {type=NiagaraFloat, defaultValue=0, description=""}]
}
